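; Effects actor for Scooter's recon scan, presenting the scan pulse through Niagara and audio.
; BeginPlay sizes the pulse to twice the scalable Radius, starts an attached scanning loop, and plays two pings two seconds apart.
; Each ping sets IsFriendly from the instigator's team affiliation to the local player.
; On destruction it fades the loop, then requests pulse deactivation after a Delay, which may never fire on the destroyed actor.
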(event Collision|EventActorBeginOverlap
  (params (OtherActor :type Actor%20Object%20Reference)))

(event EventDestroyed
  (bind _fade_out_duration (Variables|Default|GetFade%20Out%20Duration))
  (Audio|Components|Audio|FadeOut (Variables|Default|GetScanningLoopAudio) _fade_out_duration 0.0 "Logarithmic" :!class /Script/Engine.AudioComponent)
  (Utilities|FlowControl|Delay _fade_out_duration)
  (Components|Activation|Deactivate (Variables|Default|GetNS_Scooter_ScanPulse)))

(event EventBeginPlay
  ;[1 0.9524 0] Temp Audio for pulse scan sound
  (Niagara|SetNiagaraVariable(Float) (Variables|Default|GetNS_Scooter_ScanPulse) "SphereRadius" (* (ScalableFloat|GetValueatLevel (Variables|Default|GetRadius)) 2.0))
  (CallFunction|PlayPulseLoop)
  (CallFunction|PlayPingSound)
  (Utilities|FlowControl|Delay 2.0)
  (CallFunction|PlayPingSound)
  (Utilities|FlowControl|Delay 1.5))

(fn PlayPingSound
  (bind _returnvalue (Audio|SpawnSoundAttached (Variables|Default|GetPingSound) (Variables|Default|GetDefaultSceneRoot)))
  (Audio|Parameter|SetBooleanParameter _returnvalue "IsFriendly" (Utilities|Enum|Equal(Enum) (Team|GetTeamAffiliationtoLocalPlayer (Team|GetActorTeam (Game|GetInstigator))) "Friendly") :!class /Script/AudioExtensions.AudioParameterControllerInterface))

(fn PlayPulseLoop
  (bind _returnvalue (Audio|SpawnSoundAttached "/ScooterGameplay/Sounds/ScooterRecon/ScooterRecon_Projectile_Scanning_GroundLoop_Cue.ScooterRecon_Projectile_Scanning_GroundLoop_Cue" (Variables|Default|GetDefaultSceneRoot)))
  (Variables|Default|SetScanningLoopAudio _returnvalue))

(fn ConstructionScript)
