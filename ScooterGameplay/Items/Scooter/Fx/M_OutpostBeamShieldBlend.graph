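# Translucent unlit Niagara mesh-particle material for a Scooter beam shield effect, with glowing scrolling lines and refraction.
# Emissive sums V-gradient glow, noise-distorted stripes and a panning outline texture, tinted by particle color.
# Dynamic parameters drive a noisy erode wipe on opacity and a refraction amount blended by depth behind the surface.
# A separate object-scale-tiled caustic branch from two panning normal maps feeds neither emissive nor opacity.
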Material /ScooterGameplay/Items/Scooter/Fx/M_OutpostBeamShieldBlend
graph {
 properties {
  bUsedWithNiagaraMeshParticles=true
  blendMode=BLEND_Translucent
  pixelDepthOffsetMode=PDOM_Legacy
  refractionMethod=RM_IndexOfRefraction
  shadingModel=MSM_Unlit
  usedShadingModels=Unlit
 }
 nodes [
  n1 = LinearInterpolate(
   B=Multiply(
    A=ScalarParameter() {defaultValue=1.02, parameterName=RefractionAmount}
    B=DynamicParameter() {paramNames [Erode,Refraction,Param3,Param4]}.Refraction
   )
   Alpha=MaterialFunctionCall!(
    Distance Bias=ScalarParameter() {parameterName=Refraction Bias}
    Distance Divisor=ScalarParameter() {defaultValue=500, parameterName=Refraction Divisor}
    Scene Depth
    Pixel Depth
   ) {materialFunction=/Game/Effects/Fort_Effects/Materials/_Functions/MF_DistanceBehindTranslucency_01}.Normalized
  ) {constA=1}
  LinearInterpolate(B=@n1, Alpha=Saturate()) {constA=1}
  n2 = ComponentMask(None=TextureCoordinate()) {g=true}
  n3 = Multiply(A=Power(Base=OneMinus(None=@n2)) {constExponent=64}) {constB=8}
  n4 = TextureSample(UVs=Panner(
    Coordinate=TextureCoordinate() {uTiling=2, vTiling=2}
    Speed=VectorParameter() {parameterName=Texture Panner Speed, defaultValue=<rgba -0.05 0.1 0 1>}.RGB
   ) {speedX=-0.05, speedY=0.1}) {texture=&T_NoisePack_01}
  n5 = Subtract(A=Multiply(A=MaterialFunctionCall(X=@n4.R, Y=@n4.R) {materialFunction=&MakeFloat2}.Result) {constB=2})
  n6 = Panner(
   Coordinate=Add(A=TextureCoordinate(), B=Multiply(A=@n5) {constB=0.005}) {constB=0}
   Time=Multiply(A=ScalarParameter() {defaultValue=-0.05, parameterName=Line Panner Speed}, B=Time())
  ) {bFractionalPart=true, speedY=1}
  n7 = ScalarParameter() {defaultValue=50, parameterName=Line Repeat}
  n8 = ComponentMask(None=Frac(None=Multiply(A=@n6, B=@n7))) {g=true}
  n9 = Multiply(A=@n8, B=OneMinus(None=@n8))
  n10 = ScalarParameter() {defaultValue=-0.245, parameterName=Line Ceil Bias}
  n11 = Ceil(None=Clamp(None=Add(A=@n9, B=@n10)))
  n12 = VectorParameter() {parameterName=Lines Color, defaultValue=<rgba 0.58 1.6146 5 1>}
  n13 = ComponentMask(None=TextureCoordinate()) {g=true}
  n14 = Multiply(A=Power(Base=OneMinus(None=@n13)) {constExponent=8}, B=ScalarParameter() {defaultValue=1, parameterName=CenterGlow})
  n15 = VectorParameter() {parameterName=Colour, defaultValue=<rgba 0.616 2.0773 5 0.5>}
  n16 = ScalarParameter() {defaultValue=0.01, parameterName=DistortStrength}
  n17 = MaterialFunctionCall(Float2=TextureCoordinate() {vTiling=3}) {materialFunction=&BreakOutFloat2Components}
  n18 = MaterialFunctionCall(X=@n17.G, Y=@n17.G) {materialFunction=&MakeFloat2}
  n19 = VectorParameter() {parameterName=Texture Panner Speed2, defaultValue=<rgba -0.2 0.05 0 1>}
  n20 = TextureSample(UVs=Add(A=Multiply(A=@n5, B=@n16) {constB=0.2}, B=Panner(Coordinate=@n18.Result, Speed=@n19.RGB) {speedX=-0.2, speedY=0.05})) {texture=/Game/TheOutpost/Textures/T_OutpostBeamOutline_01}
  n21 = Add(A=Multiply(A=Multiply(A=@n11, B=@n12.RGB)) {constB=0.3}, B=Add(A=Multiply(A=@n14, B=@n15.RGB), B=Multiply(A=@n20.RGB, B=@n15.RGB)))
  n22 = ParticleColor()
  MaterialFunctionCall(Emissive=Multiply(A=Add(A=@n3, B=@n21), B=@n22.RGB)) {materialFunction=&MF_FX_EmissiveAdjustments}
  n23 = OneMinus(None=ComponentMask(None=TextureCoordinate()) {g=true})
  n24 = Multiply(A=Multiply(A=@n20.A, B=@n15.A), B=Multiply(A=@n23, B=@n23))
  n25 = ComponentMask(None=Add(A=Multiply(A=@n5) {constB=0.03}, B=TextureCoordinate())) {g=true}
  n26 = Subtract(
   A=Multiply(A=OneMinus(None=@n25)) {constB=8}
   B=LinearInterpolate(Alpha=DynamicParameter() {paramNames [Erode,Refraction,Param3,Param4]}.Erode) {constA=8, constB=-1}
  ) {constB=7}
  Multiply(A=@n22.A, B=Multiply(A=Add(A=@n14, B=@n24), B=Clamp(None=@n26)))
  n27 = MaterialFunctionCall() {materialFunction=/Engine/Functions/Engine_MaterialFunctions02/WorldPositionOffset/ObjectScale}
  n28 = Multiply(A=Divide(A=Round(None=@n27.Scale X), B=ScalarParameter() {defaultValue=2, parameterName=Tile UV Scale}), B=TextureCoordinate())
  n29 = AppendVector(
   A=ScalarParameter() {defaultValue=2, parameterName=U Repeat Caustic 1}
   B=ScalarParameter() {defaultValue=1, parameterName=V Repeat Caustic 1}
  )
  n30 = TextureSample(UVs=Panner(Coordinate=Multiply(A=@n28, B=@n29)) {bFractionalPart=true, speedX=0.0268, speedY=-0.0194}) {samplerType=SAMPLERTYPE_Normal, texture=/Game/TheOutpost/Textures/TilingWeakPointNormal}
  n31 = Time() {bOverride_Period=true, period=7.5075}
  n32 = ScalarParameter() {defaultValue=0.1332, parameterName=Sine Speed A}
  n33 = Add(A=@n30.RGB, B=ConstantBiasScale(None=Multiply(A=@n31, B=@n32)) {bias=-0.5, scale=2.2})
  n34 = Time() {bOverride_Period=true, period=15.9185}
  n35 = ScalarParameter() {defaultValue=0.0628, parameterName=Sine Speed B}
  n36 = Multiply(
   A=@n28
   B=AppendVector(
    A=ScalarParameter() {defaultValue=1, parameterName=U Repeat Caustic 2}
    B=ScalarParameter() {defaultValue=0.5312, parameterName=V Repeat Caustic 2}
   )
  )
  n37 = Add(
   A=ConstantBiasScale(None=Multiply(A=@n34, B=@n35)) {bias=-0.5, scale=4.4}
   B=TextureSample(UVs=Panner(Coordinate=@n36) {bFractionalPart=true, speedX=-0.0226, speedY=-0.0104}) {samplerType=SAMPLERTYPE_Normal, texture=&TilingWeakPointNormal}.RGB
  )
  n38 = Desaturation(None=Multiply(A=Sine(None=@n33) {period=2.2}, B=Sine(None=@n37) {period=4.4})) {luminanceFactors=<rgba 1 1 1 1>}
  n39 = Abs(None=@n38)
  n40 = Power(Base=Saturate(None=OneMinus(None=@n39)), Exp=ScalarParameter() {defaultValue=5, parameterName=Caustic Exponent})
  Multiply(A=LinearInterpolate(A=@n40, B=Round(None=@n40)))
  OneMinus(None=ConstantBiasScale(None=@n38) {bias=2, scale=0.25})
  Reroute(None=Normalize(VectorInput=Add(A=@n30.RGB)))
  Reroute(None=@n1)
 ]
}
