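# Additive, unlit, two-sided Niagara material for a Scooter item's Recon cylinder effect, tinted by particle color.
# Panning dissolve and blurred noise, pulsed by sine waves, form thin threshold bands combined with Fresnel and depth-fade masks.
# World position offset scrolls a distortion texture, scaled by camera-distance LOD with separate mobile and glider variants.
# An optional static switch adds a UV-weighted bend driven by Niagara dynamic parameters and particle position.
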Material /ScooterGameplay/Items/Scooter/Fx/M_Recon_IO_Cylinder
graph {
 properties {
  bEnableResponsiveAA=true
  bUseTranslucencyVertexFog=false
  bUsedWithNiagaraMeshParticles=true
  bUsedWithNiagaraRibbons=true
  bUsedWithNiagaraSprites=true
  blendMode=BLEND_Additive
  pixelDepthOffsetMode=PDOM_Legacy
  shadingModel=MSM_Unlit
  twoSided=true
  usedShadingModels=Unlit
 }
 nodes [
  n1 = TextureSample(UVs=Panner() {speedX=-0.5, speedY=-0.5}) {texture=/Game/CharClasses/Outlander/Gadgets/AntiMaterialCharge/Materials/T_AntiMaterialCharge_Dissolve_01}
  n2 = Sine(None=Time() {period=1}) {period=4}
  n3 = Add(A=@n1.R, B=@n2)
  Add(A=@n1.R, B=@n3) {constB=0.5}
  n4 = DynamicParameter() {desc=Parameters set by Niagara, parameterIndex=1, paramNames [X,Y,Z,WPOIntensity]}
  n5 = Clamp(
   None=Subtract(
    A=MaterialFunctionCall(X=@n4.X, Y=@n4.X, Z=@n4.X) {materialFunction=&MakeFloat3}.Result
    B=ParticlePositionWS()
   )
   Min=ScalarParameter() {defaultValue=-100, group=WPO, parameterName=Clamp_min}
   Max=ScalarParameter() {defaultValue=100, group=WPO, parameterName=Clamp_max}
  ) {maxDefault=1000, minDefault=-1000}
  n6 = OneMinus(None=TextureCoordinate())
  n7 = Multiply(
   A=Power(Base=ComponentMask(None=@n6) {g=true}, Exp=ScalarParameter() {defaultValue=3, group=WPO, parameterName=WPOBendPower})
   B=@n4.WPOIntensity
  )
  n8 = OneMinus(None=MaterialFunctionCall(Distance=Constant() {r=3000}) {materialFunction=&Near_Camera_Fade}.Result)
  n9 = VectorParameter() {group=WPO, parameterName=WPO_UVScale, defaultValue=<rgba 0 1 0 0>}
  n10 = Add(A=Multiply(A=TextureCoordinate(), B=ComponentMask(None=@n9.RGB) {g=true, r=true}), B=AppendVector(A=Constant(), B=VertexColor().R)) {constB=0}
  n11 = ComponentMask(None=VectorParameter() {group=WPO, parameterName=WPO_Speed, defaultValue=<rgba 0 0.5 0 0>}.RGB) {g=true, r=true}
  n12 = LinearInterpolate(Alpha=TextureSample(UVs=Panner(Coordinate=@n10, Speed=@n11) {speedY=0.5}) {samplerType=SAMPLERTYPE_Normal, texture=&T_Weap_Distort_N_02}.R) {constB=1.5}
  n13 = Transform(None=Multiply(A=Multiply(A=@n12), B=VectorParameter() {group=WPO, parameterName=WPO_direction, defaultValue=<rgba 1 0 0 1>}.RGB))
  n14 = Multiply(
   A=@n13
   B=LinearInterpolate(
    A=ScalarParameter() {defaultValue=20, group=WPO, parameterName=Glider_Lod_low_WPO_intensity}
    B=ScalarParameter() {defaultValue=80, group=WPO, parameterName=Glider_WPO_intensity}
    Alpha=Saturate(None=@n8)
   )
  ) {constB=80}
  n15 = OneMinus(None=MaterialFunctionCall(Distance=Constant() {r=1000}) {materialFunction=&Near_Camera_Fade}.Result)
  n16 = Multiply(
   A=@n13
   B=LinearInterpolate(
    A=ScalarParameter() {defaultValue=15, group=WPO, parameterName=Glider_Moble_Lod_low_WPO_intensity}
    B=ScalarParameter() {group=WPO, parameterName=Glider_WPO_intensity}
    Alpha=Saturate(None=@n15)
   )
  )
  n17 = OneMinus(None=MaterialFunctionCall(Distance=Constant() {r=1700}) {materialFunction=&Near_Camera_Fade}.Result)
  n18 = Multiply(
   A=@n13
   B=LinearInterpolate(
    A=ScalarParameter() {defaultValue=20, group=WPO, parameterName=Lod_low_WPO_intensity}
    B=ScalarParameter() {defaultValue=80, group=WPO, parameterName=WPO_intensity}
    Alpha=Saturate(None=@n17)
   )
  ) {constB=80}
  n19 = OneMinus(None=MaterialFunctionCall(Distance=Constant() {r=500}) {materialFunction=&Near_Camera_Fade}.Result)
  n20 = Multiply(
   A=@n13
   B=LinearInterpolate(
    A=ScalarParameter() {defaultValue=15, group=WPO, parameterName=Moble_Lod_low_WPO_intensity}
    B=ScalarParameter() {group=WPO, parameterName=WPO_intensity}
    Alpha=Saturate(None=@n19)
   )
  )
  Add(
   A=StaticSwitchParameter(True=LinearInterpolate(B=@n5, Alpha=@n7), False=Constant()) {group=WPO, parameterName=Use_Interpolate WPO}
   B=StaticSwitchParameter(
    True=ShadingPathSwitch(Default=@n14, Deferred=@n14, Forward=@n14, Mobile=@n16)
    False=ShadingPathSwitch(Default=@n18, Deferred=@n18, Forward=@n18, Mobile=@n20)
   ) {group=WPO, parameterName=Use_Glider_Lod_WPO_Fix}
  )
  n21 = TextureSample(UVs=Panner() {speedX=-0.2, speedY=-0.2}) {texture=&T_AntiMaterialCharge_Dissolve_01}
  n22 = Sine(None=Time() {period=1}) {period=5}
  n23 = Add(
   A=@n21.R
   B=MaterialFunctionCall!(Sine Phase=@n22, Value 1=Constant() {r=0.2}, Value 2=Constant() {r=0.8}) {materialFunction=/Engine/Functions/Engine_MaterialFunctions02/Utility/Sine_Remapped}.Remapped Sine
  )
  Add(A=@n21.R, B=@n23) {constB=0.5}
  Frac(None=@n22)
  Frac(None=@n2)
  n24 = Panner() {speedX=0.5, speedY=0.5}
  n25 = Add(
   A=TextureSample(UVs=Panner() {speedX=0.3, speedY=0.3}) {samplerType=SAMPLERTYPE_Grayscale, texture=&LowResBlurredNoise}.R
   B=@n23
  )
  n26 = Multiply(
   A=Add(A=TextureSample(UVs=@n24) {samplerType=SAMPLERTYPE_Grayscale, texture=&LowResBlurredNoise}.R, B=@n3)
   B=@n25
  )
  LinearInterpolate(Alpha=@n26)
  LinearInterpolate()
  n27 = ParticleColor()
  MaterialFunctionCall(Emissive=@n27.RGB) {materialFunction=&MF_FX_EmissiveAdjustments}
  n28 = Constant()
  n29 = Constant()
  n30 = Add(
   A=If(A=@n26, B=Constant() {r=1}, "A > B"=@n28, "A == B"=Constant() {r=1}, "A < B"=@n28) {equalsThreshold=0.01}
   B=If(A=@n25, B=Constant() {r=1}, "A > B"=@n29, "A == B"=Constant() {r=1}, "A < B"=@n29) {equalsThreshold=0.05}
  )
  n31 = DynamicParameter() {defaultValue=<rgba 0 1 1 1>, paramNames [FresnelMultExp, FresnelMultIn, FresnelMultOut, Param4]}
  n32 = ScalarParameter() {defaultValue=1, parameterName=Fresnel Exponent}
  n33 = LinearInterpolate(A=@n31.FresnelMultIn, B=@n31.FresnelMultIn, Alpha=Fresnel(ExponentIn=Multiply(A=@n31.FresnelMultExp, B=@n32)) {baseReflectFraction=0})
  n34 = LinearInterpolate(Alpha=Power(Base=DepthFade()) {constExponent=0.25}) {constA=1, constB=0}
  n35 = Clamp(None=DepthFade(Opacity=Add(A=@n33, B=@n34)) {fadeDistanceDefault=20})
  Multiply(A=@n27.A, B=Add(A=Multiply(A=@n30, B=@n35), B=@n35))
  Multiply(A=@n27.A, B=OneMinus(None=@n35))
  n36 = OneMinus(None=ComponentMask(None=TextureCoordinate()) {g=true})
  Saturate(None=Multiply(A=Power(Base=@n36) {constExponent=1.5}))
 ]
}
